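; Season 24 Battle Pass quest dummy target (NightHawk) that players damage to progress a quest objective.
; On server damage, it casts the instigator to a player controller and passes the hit to parent logic only if that player has the quest and has not completed this objective.
; The construction script destroys the inherited discover-radius collision.
; A split-screen-aware local-controller check is defined but not called here; post-interaction overrides just call the parent.
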(event EventOnDamageServer
  (params (Damage :type Float%20%28single-precision%29) (DamageTags :type Gameplay%20Tag%20Container%20Structure%20%28by%20ref%29) (Momentum :type Vector :default "0, 0, 0") (HitInfo :type Hit%20Result%20Structure%20%28by%20ref%29) (InstigatedBy :type Controller%20Object%20Reference) (DamageCauser :type Actor%20Object%20Reference) (EffectContext :type Gameplay%20Effect%20Context%20Handle%20Structure))
  ;[0 0 0] Server only - logic for when a player hits a target
  ;;[0.1719 0.1719 0.1719] Ensure that a player hit the target, and that they have the quest and objective assoicated with this target
  (bind _asfort_player_controller (Utilities|Casting|CastToFortPlayerController InstigatedBy))
  (bind _bcanplay (CallFunction|InstigatorCanPlay _asfort_player_controller))
  (if _bcanplay
    ;;[0.1719 0.1719 0.1719] Send Success to Quest Manager
    (|CallParentFunction Damage DamageTags Momentum HitInfo InstigatedBy DamageCauser EffectContext)))

(fn GetOwnedTags
  (params (self2 :type Gameplay%20Tag%20Asset%20Interface%20Interface))
  (return (GameplayTags|GetOwnedGameplayTags self2 :!class /Script/GameplayTags.GameplayTagAssetInterface)))

(fn InstigatorCanPlay
  (params (InstigatingPlayerController :type Fort%20Player%20Controller%20Object%20Reference))
  (bind _consolidatedquestcomponent (Variables|BuildingGameplayActorPropQuest|GetConsolidatedQuestComponent))
  (bind _returnvalue (Class|ConsolidatedQuestComponent|GetQuestItemDefinition _consolidatedquestcomponent))
  ;[0 0 0] Check if this player is on the quest and they have not already completed the objective for this target
  (Utilities|IsValid InstigatingPlayerController
    (:Is%20Valid
      (Utilities|IsValid _returnvalue
        (:Is%20Valid
          (Utilities|IsValid (Quests|GetQuestManager InstigatingPlayerController "Invalid")
            (:Is%20Valid
              (bind _playerhasquest (Class|AthenaQuestFunctionLibrary|AthenaHasQuest InstigatingPlayerController _returnvalue))
              (if _playerhasquest
                (bind (_playercontrollerout _completedquest) (Quest|HasPlayerCompletedQuest InstigatingPlayerController _returnvalue (Variables|/Script/FortniteGame.ConsolidatedQuestComponent|GetObjectiveBackendName (Variables|BuildingGameplayActorPropQuest|GetConsolidatedQuestComponent))))
                (if (not _completedquest)
                  (return true)
                  (else
                    _))
                (else
                  _))
              (return false))
            (:Is%20Not%20Valid
              _)))
        (:Is%20Not%20Valid
          _)))
    (:Is%20Not%20Valid
      _)))

(fn IsValidLocalPlayerController
  (params (InstigatingPlayerController :type Fort%20Player%20Controller%20Object%20Reference))
  ;[0 0 0] Returns true on the executing client if the instigator controller is locally controlled. Returns false if running on the server
  (if (not (Networking|IsServer))
    (Utilities|IsValid InstigatingPlayerController
      (:Is%20Valid
        (if (PlayerController|IsinSplitscreen InstigatingPlayerController)
          (for _array_element (Fortnite|GetLocalFortPlayerControllers true)
            (if (== _array_element InstigatingPlayerController)
              (return true)))
          (return false)
          (else
            (return true))))
      (:Is%20Not%20Valid
        _)))
  (return false))

(fn PostInteractionVFX&Audio
  (params (PlayerController :type Fort%20Player%20Controller%20Object%20Reference))
  (|CallParentFunction PlayerController))

(fn PostInteractionVisuals
  (params (PlayerController :type Fort%20Player%20Controller%20Object%20Reference))
  (|CallParentFunction PlayerController))

(fn ConstructionScript
  (|CallParentFunction)
  (Components|DestroyComponent (Variables|Components|GetCollision_DiscoverRadius)))
